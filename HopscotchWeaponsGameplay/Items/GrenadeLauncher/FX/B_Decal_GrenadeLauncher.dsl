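; Decal Blueprint for the Grenade Launcher effects, showing an explosion decal material.
; The construction script creates a dynamic instance of MID_Decal_Explosion, stores it in the Decal_MID variable and applies it to the Decal component.
; BeginPlay is empty, and no material parameters are changed here.
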(event EventBeginPlay)

(fn ConstructionScript
  (bind _returnvalue (Rendering|Material|CreateDynamicMaterialInstance "/HopscotchWeaponsGameplay/Items/GrenadeLauncher/FX/MID_Decal_Explosion.MID_Decal_Explosion" "None" "None" :!class /Script/Engine.KismetMaterialLibrary))
  (bind _output_get (Variables|Default|SetDecal_MID _returnvalue))
  (Rendering|Components|Decal|SetDecalMaterial (Variables|Decal|GetDecal) _output_get))
