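# Translucent deferred decal material for grenade launcher explosions, combining a ground char mark with a glowing, cooling scorch.
# The glow masks a scorch-hole texture with an age-driven power curve and edge-to-center HDR color blend, behind an emissive switch.
# Age comes from decal lifetime opacity or a spawn-time timestamp, with a looping debug option.
# The char texture, scaled by Opacity, is eroded by tiling smoke noise and faded over decal lifetime.
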Material /HopscotchWeaponsGameplay/Items/GrenadeLauncher/FX/MD_Decal_Explosion
graph {
 properties {
  bAutomaticallySetUsageInEditor=false
  blendMode=BLEND_Translucent
  materialDomain=MD_DeferredDecal
  pixelDepthOffsetMode=PDOM_Legacy
  shadingModel=MSM_Unlit
  usedShadingModels=Unlit
 }
 nodes [
  n1 = StaticSwitchParameter(True=Frac(None=Time() {period=1}), False=Subtract(A=Time(), B=ScalarParameter() {group=Timestamp, parameterName=Spawntime})) {group=Timestamp, parameterName=Debug Timestamp}
  n2 = ScalarParameter() {defaultValue=1, group=Timestamp, parameterName=Lifetime}
  n3 = StaticSwitchParameter(True=DecalLifetimeOpacity().Opacity, False=OneMinus(None=Divide(A=@n1, B=@n2))) {parameterName=UseDecalLifetimeOpacity}
  n4 = Clamp(None=LinearInterpolate(A=ScalarParameter() {defaultValue=-4, parameterName=GlowOffset}, Alpha=Clamp(None=@n3)) {constA=-7})
  n5 = Power(Base=@n4)
  n6 = TextureSample() {texture=/Game/Effects/Fort_Effects/Gameplay/SmokeDustExplosions/Textures/T_Decal_Scorch_Hole}
  n7 = LinearInterpolate(Alpha=@n4) {constA=14}
  n8 = Multiply(
   A=LinearInterpolate(
    A=VectorParameter() {group=Emissive, parameterName=EmissiveEdge, defaultValue=<rgba 10 0.9091 0 1>}.RGB
    B=VectorParameter() {group=Emissive, parameterName=EmissiveCenter, defaultValue=<rgba 14 6.6483 1.5312 1>}.RGB
    Alpha=@n5
   )
   B=Multiply(A=@n5, B=Power(Base=@n6.R, Exp=@n7))
  )
  n9 = Clamp(None=@n8)
  MaterialFunctionCall(Emissive=StaticSwitchParameter(
    True=MaterialFunctionCall(LDR Input=@n9, HDR Tint=@n8) {materialFunction=&DeriveHDRfromLDR}.RemappedRange
    False=Constant()
   ) {group=Timestamp, parameterName=Use Emissive?}) {materialFunction=&MF_FX_EmissiveAdjustments}
  n10 = TextureSampleParameter2D() {
   parameterName=DecalTexture
   samplerType=SAMPLERTYPE_LinearColor
   texture=/Game/Effects/Fort_Effects/Gameplay/SmokeDustExplosions/Textures/T_GroundChar_01
  }
  n11 = ScalarParameter() {defaultValue=0.8, parameterName=Opacity}
  n12 = TextureObjectParameter() {group=Erosion, parameterName=Erosion_Texture, sortPriority=0, texture=&T_Inky_Smoke_Tile}
  n13 = VectorParameter() {group=Erosion, parameterName=Erosion_UVControl, channelNames {a=VPan, b=UPan, g=VScale, r=UScale}, defaultValue=<rgba 1.5 1.5 0 0>}
  n14 = MaterialFunctionCall(Texture=@n12, UVControl_RGB=@n13, UVControl_A=@n13) {materialFunction=&FXCommon_TextureSample}
  n15 = DecalLifetimeOpacity()
  n16 = OneMinus(None=@n15.Opacity)
  n17 = ScalarParameter() {defaultValue=0.7, parameterName=Softness}
  Multiply(
   A=Multiply(
    A=LinearInterpolate(B=@n10.RGB, Alpha=@n11) {constB=0}
    B=MaterialFunctionCall!(Texture=@n14.R, Erosion=@n16, Softness=@n17) {materialFunction=/Game/Effects/Niagara/MasterMaterials/Functions/MF_Erosion_Smoothstep_01}.Result
   )
   B=@n15.Opacity
  )
  ScalarParameter() {parameterName=Metallic}
  ScalarParameter() {parameterName=Specular}
  ScalarParameter() {defaultValue=1, parameterName=Roughness}
  VectorParameter() {parameterName=DecalColor, defaultValue=<rgba 0 0 0 1>}
 ]
}
