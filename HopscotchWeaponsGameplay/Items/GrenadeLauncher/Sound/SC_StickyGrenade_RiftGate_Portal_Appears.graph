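# Sound cue for a rift gate portal appearing from the grenade launcher's sticky grenade.
# A random node picks one of eight equally weighted Lager_Spawn_Rift waves.
# Overridden attenuation gives logarithmic falloff with distance-based low-pass filtering and reverb send settings, and overridden concurrency caps simultaneous plays per owner.
# It routes to the gadgets and items sound class and keeps playing when silent.
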SoundCue /HopscotchWeaponsGameplay/Items/GrenadeLauncher/Sound/SC_StickyGrenade_RiftGate_Portal_Appears
root=Random_0
nodes [
 {
  id=Random_0
  children [WavePlayer_1,WavePlayer_2,WavePlayer_3,WavePlayer_4,WavePlayer_5,WavePlayer_6,WavePlayer_7,WavePlayer_8]
  properties {weights [1,1,1,1,1,1,1,1]}
 }
 {id=WavePlayer_1, properties {soundWaveAssetPtr=&/Lager_Spawn_Rift_08}}
 {id=WavePlayer_2, properties {soundWaveAssetPtr=&/Lager_Spawn_Rift_07}}
 {id=WavePlayer_3, properties {soundWaveAssetPtr=&/Lager_Spawn_Rift_06}}
 {id=WavePlayer_4, properties {soundWaveAssetPtr=&/Lager_Spawn_Rift_05}}
 {id=WavePlayer_5, properties {soundWaveAssetPtr=&/Lager_Spawn_Rift_04}}
 {id=WavePlayer_6, properties {soundWaveAssetPtr=&/Lager_Spawn_Rift_03}}
 {id=WavePlayer_7, properties {soundWaveAssetPtr=&/Lager_Spawn_Rift_02}}
 {id=WavePlayer_8, properties {soundWaveAssetPtr=&/Lager_Spawn_Rift_01}}
]
properties {
 bOverrideAttenuation=true
 bOverrideConcurrency=true
 maxDistance=8512
 soundClassObject=&Fort_Gadgets_Items
 virtualizationMode=PlayWhenSilent
 volumeMultiplier=1
 attenuationOverrides {
  bAttenuateWithLPF=true
  distanceAlgorithm=Logarithmic
  falloffDistance=8000
  lPFFrequencyAtMax=400
  lPFRadiusMin=2000
  reverbDistanceMax=6000
  reverbWetLevelMax=0.7
  stereoSpread=2000
  attenuationShapeExtents {x=512}
 }
 concurrencyOverrides {bLimitToOwner=true, maxCount=3}
}
