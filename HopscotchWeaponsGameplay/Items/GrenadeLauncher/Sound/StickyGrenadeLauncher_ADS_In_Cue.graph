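# Sound cue for aiming down sights with the sticky grenade launcher, classed as weapon foley.
# Mixes a quieter random player-movement foley layer with a louder, pitch-modulated launcher ADS-in layer.
# A quality-level node gives the launcher layer six random variations in its first two slots and the last three of those in the others.
# Attenuation and concurrency are overridden, allowing one instance that stops the oldest.
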SoundCue /HopscotchWeaponsGameplay/Items/GrenadeLauncher/Sound/StickyGrenadeLauncher_ADS_In_Cue
root=Mixer_0
nodes [
 {id=Mixer_0, children [Modulator_0,Modulator_1], properties {inputVolume [0.4,1]}}
 {id=Modulator_0, children [Random_4], properties {volumeMax=0.8, volumeMin=0.8}}
 {id=Modulator_1, children [QualityLevel_0], properties {pitchMax=1, volumeMax=2.5, volumeMin=2.5}}
 {id=QualityLevel_0, children [Random_0,Random_0,Random_1,Random_1]}
 {id=Random_0, children [WavePlayer_4,WavePlayer_5,WavePlayer_6,WavePlayer_7,WavePlayer_8,WavePlayer_9], properties {weights [1,1,1,1,1,1]}}
 {id=Random_1, children [WavePlayer_7,WavePlayer_8,WavePlayer_9], properties {bShouldExcludeFromBranchCulling=true, weights [1,1,1]}}
 {id=Random_4, children [WavePlayer_0,WavePlayer_1,WavePlayer_2,WavePlayer_3], properties {weights [1,1,1,1]}}
 {id=WavePlayer_0, properties {soundWaveAssetPtr=/Game/Sounds/Fort_Foley_Sounds/Moves/player_run_foley_15}}
 {id=WavePlayer_1, properties {soundWaveAssetPtr=/Game/Sounds/Fort_Foley_Sounds/Moves/player_run_foley_14}}
 {id=WavePlayer_2, properties {soundWaveAssetPtr=/Game/Sounds/Fort_Foley_Sounds/Moves/player_run_foley_13}}
 {id=WavePlayer_3, properties {soundWaveAssetPtr=/Game/Sounds/Fort_Foley_Sounds/Moves/player_run_foley_12}}
 {id=WavePlayer_4, properties {soundWaveAssetPtr=&/StickyGrenadeLauncher_ADS_In_01}}
 {id=WavePlayer_5, properties {soundWaveAssetPtr=&/StickyGrenadeLauncher_ADS_In_02}}
 {id=WavePlayer_6, properties {soundWaveAssetPtr=&/StickyGrenadeLauncher_ADS_In_03}}
 {id=WavePlayer_7, properties {soundWaveAssetPtr=&/StickyGrenadeLauncher_ADS_In_04}}
 {id=WavePlayer_8, properties {soundWaveAssetPtr=&/StickyGrenadeLauncher_ADS_In_05}}
 {id=WavePlayer_9, properties {soundWaveAssetPtr=&/StickyGrenadeLauncher_ADS_In_06}}
]
properties {
 attenuationSettings=&Fort_Weapon_Projectile_FirstPerson_Athena_Att
 bOverrideAttenuation=true
 bOverrideConcurrency=true
 maxDistance=2700
 pitchMultiplier=1.02
 soundClassObject=&Fort_WeaponFoley
 volumeMultiplier=0.1
 attenuationOverrides {falloffDistance=1500, nonSpatializedRadiusStart=200, attenuationShapeExtents {x=1200}}
 concurrencyOverrides {maxCount=1, resolutionRule=StopOldest}
}
