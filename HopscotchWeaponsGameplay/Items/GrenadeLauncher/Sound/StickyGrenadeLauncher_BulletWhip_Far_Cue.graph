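# Sound cue for the distant whip of a sticky grenade launcher projectile, using shared bullet-whip attenuation and whiz-by concurrency.
# A mixer layers random distant whip waves with louder, pitched-up random incoming waves, all under reduced cue-wide pitch and volume.
# Each layer switches by quality level between a five-wave pool and a three-wave subset, which is excluded from branch culling.
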SoundCue /HopscotchWeaponsGameplay/Items/GrenadeLauncher/Sound/StickyGrenadeLauncher_BulletWhip_Far_Cue
root=Mixer_0
nodes [
 {id=Mixer_0, children [Modulator_1,Modulator_0], properties {inputVolume [1,3.5]}}
 {id=Modulator_0, children [QualityLevel_0], properties {pitchMax=1.5, pitchMin=1.5}}
 {id=Modulator_1, children [QualityLevel_1], properties {pitchMax=1, pitchMin=1, volumeMax=1, volumeMin=1}}
 {id=QualityLevel_0, children [Random_0,Random_0,Random_1,Random_1]}
 {id=QualityLevel_1, children [Random_2,Random_2,Random_3,Random_3]}
 {id=Random_0, children [WavePlayer_2,WavePlayer_6,WavePlayer_7,WavePlayer_8,WavePlayer_9], properties {weights [1,1,1,1,1]}}
 {id=Random_1, children [WavePlayer_7,WavePlayer_8,WavePlayer_9], properties {bShouldExcludeFromBranchCulling=true, weights [1,1,1]}}
 {id=Random_2, children [WavePlayer_0,WavePlayer_1,WavePlayer_3,WavePlayer_4,WavePlayer_5], properties {weights [1,1,1,1,1]}}
 {id=Random_3, children [WavePlayer_3,WavePlayer_4,WavePlayer_5], properties {bShouldExcludeFromBranchCulling=true, weights [1,1,1]}}
 {
  id=WavePlayer_0
  properties {soundWaveAssetPtr=&/StickyGrenadeLauncher_GrenadeWhip_Distant_01}
 }
 {
  id=WavePlayer_1
  properties {soundWaveAssetPtr=&/StickyGrenadeLauncher_GrenadeWhip_Distant_02}
 }
 {id=WavePlayer_2, properties {soundWaveAssetPtr=&/1StickyGrenadeLauncher_Incoming_01}}
 {
  id=WavePlayer_3
  properties {soundWaveAssetPtr=&/StickyGrenadeLauncher_GrenadeWhip_Distant_04}
 }
 {
  id=WavePlayer_4
  properties {soundWaveAssetPtr=&/StickyGrenadeLauncher_GrenadeWhip_Distant_05}
 }
 {
  id=WavePlayer_5
  properties {soundWaveAssetPtr=&/StickyGrenadeLauncher_GrenadeWhip_Distant_06}
 }
 {id=WavePlayer_6, properties {soundWaveAssetPtr=&/1StickyGrenadeLauncher_Incoming_02}}
 {id=WavePlayer_7, properties {soundWaveAssetPtr=&/1StickyGrenadeLauncher_Incoming_03}}
 {id=WavePlayer_8, properties {soundWaveAssetPtr=&/1StickyGrenadeLauncher_Incoming_04}}
 {id=WavePlayer_9, properties {soundWaveAssetPtr=&/1StickyGrenadeLauncher_Incoming_05}}
]
properties {
 attenuationSettings=/Game/Athena/Sounds/Weapons/BulletWhips/bullet_whip_Att.Bullet_Whip_Att
 maxDistance=2097152
 pitchMultiplier=0.5
 soundClassObject=&Fort_WeaponRanged_Projectile_Effects
 volumeMultiplier=0.25
 attenuationOverrides {bAttenuate=false, stereoSpread=0}
 concurrencySet [/Game/Sounds/Concurrency_Rules/Athena_BulletWhizBy_Conc]
}
