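# First-person fire sound cue for the Sticky Grenade Launcher.
# A mixer layers random shot and quieter mechanical variants, a randomized "incoming" layer delayed a quarter second, and a launch-timer sound that plays only when IsLocal is true.
# Quality-level nodes switch the shot and incoming layers between full variant pools and three-sound subsets.
# First-person attenuation and gunfire concurrency settings apply.
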SoundCue /HopscotchWeaponsGameplay/Items/GrenadeLauncher/Sound/StickyGrenadeLauncher_Fire_1P_Cue
root=Modulator_0
nodes [
 {id=Branch_0, children [Delay_0,null,null], properties {boolParameterName=IsLocal}}
 {id=Delay_0, children [Modulator_2]}
 {id=Delay_1, children [Modulator_4], properties {delayMax=0.25, delayMin=0.25}}
 {id=Mixer_0, children [Modulator_3,Modulator_1,Delay_1,Branch_0], properties {inputVolume [1,0.2,1,1.3]}}
 {id=Modulator_0, children [Mixer_0], properties {pitchMax=1, pitchMin=1}}
 {id=Modulator_1, children [Random_2], properties {pitchMax=1.1, pitchMin=1.1, volumeMax=1, volumeMin=1}}
 {id=Modulator_2, children [WavePlayer_14], properties {pitchMax=1, pitchMin=1}}
 {id=Modulator_3, children [QualityLevel_0], properties {pitchMax=1, pitchMin=1, volumeMax=1.2, volumeMin=1.2}}
 {id=Modulator_4, children [QualityLevel_1], properties {pitchMax=0.95}}
 {id=QualityLevel_0, children [Random_0,Random_0,Random_1,Random_1]}
 {id=QualityLevel_1, children [Random_4,Random_4,Random_5,Random_5]}
 {
  id=Random_0
  children [WavePlayer_0,WavePlayer_1,WavePlayer_2,WavePlayer_3,WavePlayer_4,WavePlayer_5,WavePlayer_6,WavePlayer_7,WavePlayer_8,WavePlayer_9]
  properties {weights [1,1,1,1,1,1,1,1,1,1]}
 }
 {id=Random_1, children [WavePlayer_7,WavePlayer_8,WavePlayer_9], properties {bShouldExcludeFromBranchCulling=true, weights [1,1,1]}}
 {id=Random_2, children [WavePlayer_10, WavePlayer_11, WavePlayer_12, WavePlayer_13], properties {weights [1,1,1,1]}}
 {id=Random_4, children [WavePlayer_21, WavePlayer_22, WavePlayer_23, WavePlayer_24, WavePlayer_25], properties {weights [1,1,1,1,1]}}
 {id=Random_5, children [WavePlayer_23, WavePlayer_24, WavePlayer_25], properties {bShouldExcludeFromBranchCulling=true, weights [1,1,1]}}
 {id=WavePlayer_0, properties {soundWaveAssetPtr=&/StickyGrenadeLauncher_Shoot_1P_01}}
 {id=WavePlayer_1, properties {soundWaveAssetPtr=&/StickyGrenadeLauncher_Shoot_1P_02}}
 {id=WavePlayer_10, properties {soundWaveAssetPtr=&/StickyGrenadeLauncher_1P_Shoot_Mec_01}}
 {id=WavePlayer_11, properties {soundWaveAssetPtr=&/StickyGrenadeLauncher_1P_Shoot_Mec_02}}
 {id=WavePlayer_12, properties {soundWaveAssetPtr=&/StickyGrenadeLauncher_1P_Shoot_Mec_03}}
 {id=WavePlayer_13, properties {soundWaveAssetPtr=&/StickyGrenadeLauncher_1P_Shoot_Mec_04}}
 {id=WavePlayer_14, properties {soundWaveAssetPtr=&/StickyGrenadeLauncher_LaunchTimer}}
 {id=WavePlayer_2, properties {soundWaveAssetPtr=&/StickyGrenadeLauncher_Shoot_1P_03}}
 {id=WavePlayer_21, properties {soundWaveAssetPtr=&/1StickyGrenadeLauncher_Incoming_01}}
 {id=WavePlayer_22, properties {soundWaveAssetPtr=&/1StickyGrenadeLauncher_Incoming_02}}
 {id=WavePlayer_23, properties {soundWaveAssetPtr=&/1StickyGrenadeLauncher_Incoming_03}}
 {id=WavePlayer_24, properties {soundWaveAssetPtr=&/1StickyGrenadeLauncher_Incoming_04}}
 {id=WavePlayer_25, properties {soundWaveAssetPtr=&/1StickyGrenadeLauncher_Incoming_05}}
 {id=WavePlayer_3, properties {soundWaveAssetPtr=&/StickyGrenadeLauncher_Shoot_1P_04}}
 {id=WavePlayer_4, properties {soundWaveAssetPtr=&/StickyGrenadeLauncher_Shoot_1P_05}}
 {id=WavePlayer_5, properties {soundWaveAssetPtr=&/StickyGrenadeLauncher_Shoot_1P_06}}
 {id=WavePlayer_6, properties {soundWaveAssetPtr=&/StickyGrenadeLauncher_Shoot_1P_07}}
 {id=WavePlayer_7, properties {soundWaveAssetPtr=&/StickyGrenadeLauncher_Shoot_1P_08}}
 {id=WavePlayer_8, properties {soundWaveAssetPtr=&/StickyGrenadeLauncher_Shoot_1P_09}}
 {id=WavePlayer_9, properties {soundWaveAssetPtr=&/StickyGrenadeLauncher_Shoot_1P_10}}
]
properties {
 attenuationSettings=&Fort_Weapon_Projectile_FirstPerson_Athena_Att
 bDebug=true
 maxDistance=2097152
 soundClassObject=&Fort_WeaponRanged_Fire_1P_PSM
 volumeMultiplier=1
 attenuationOverrides {bSpatialize=false}
 concurrencyOverrides {bLimitToOwner=true, maxCount=3, resolutionRule=StopOldest}
 concurrencySet [&Weapon_Gunfire_FirstPerson_Conc]
}
