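# Third-person fire sound cue for the sticky grenade launcher, layering close, medium and distant shots.
# Each layer has its own attenuation, sound class override and volume modulator before mixing, with pitch variation on close and medium.
# Quality-level switches choose between ten random waves and a three-wave subset excluded from branch culling.
# Gunfire concurrency sets apply, alongside a per-owner override that stops the oldest instance.
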SoundCue /HopscotchWeaponsGameplay/Items/GrenadeLauncher/Sound/StickyGrenadeLauncher_Fire_3P_Cue
root=Mixer_2
nodes [
 {id=Attenuation_3, children [Modulator_0], properties {attenuationSettings=&Fort_Weapon_Projectile_Close_Athena_Att}}
 {
  id=Attenuation_4
  children [Modulator_2]
  properties {attenuationSettings=&Fort_Weapon_Projectile_Distant_Mono_Athena_Att}
 }
 {
  id=Attenuation_5
  children [Modulator_4]
  properties {attenuationSettings=&Fort_Weapon_Projectile_Medium_Mono_Athena_Att}
 }
 {id=Mixer_2, children [Modulator_3,SoundClass_4], properties {inputVolume [1,2.75]}}
 {id=Mixer_3, children [SoundClass_3,SoundClass_5], properties {inputVolume [2.75,4]}}
 {id=Modulator_0, children [QualityLevel_2], properties {pitchMax=1.02, pitchMin=0.98, volumeMax=0.8, volumeMin=0.8}}
 {id=Modulator_2, children [QualityLevel_4], properties {volumeMax=0.25, volumeMin=0.25}}
 {id=Modulator_3, children [Mixer_3]}
 {id=Modulator_4, children [QualityLevel_3], properties {pitchMax=0.9, pitchMin=0.85, volumeMax=0.2, volumeMin=0.2}}
 {id=QualityLevel_2, children [Random_2,Random_2,Random_5,Random_5]}
 {id=QualityLevel_3, children [Random_6,Random_6,Random_7,Random_7]}
 {id=QualityLevel_4, children [Random_8,Random_8,Random_9,Random_9]}
 {
  id=Random_2
  children [WavePlayer_0,WavePlayer_1,WavePlayer_2,WavePlayer_3,WavePlayer_4,WavePlayer_5,WavePlayer_6,WavePlayer_7,WavePlayer_8,WavePlayer_9]
  properties {weights [1,1,1,1,1,1,1,1,1,1]}
 }
 {id=Random_5, children [WavePlayer_7,WavePlayer_8,WavePlayer_9], properties {bShouldExcludeFromBranchCulling=true, weights [1,1,1]}}
 {
  id=Random_6
  children [
   WavePlayer_10
   WavePlayer_11
   WavePlayer_12
   WavePlayer_13
   WavePlayer_14
   WavePlayer_15
   WavePlayer_16
   WavePlayer_17
   WavePlayer_18
   WavePlayer_19
  ]
  properties {weights [1,1,1,1,1,1,1,1,1,1]}
 }
 {id=Random_7, children [WavePlayer_17, WavePlayer_18, WavePlayer_19], properties {bShouldExcludeFromBranchCulling=true, weights [1,1,1]}}
 {
  id=Random_8
  children [
   WavePlayer_20
   WavePlayer_21
   WavePlayer_22
   WavePlayer_23
   WavePlayer_24
   WavePlayer_25
   WavePlayer_26
   WavePlayer_27
   WavePlayer_28
   WavePlayer_29
  ]
  properties {weights [1,1,1,1,1,1,1,1,1,1]}
 }
 {id=Random_9, children [WavePlayer_27, WavePlayer_28, WavePlayer_29], properties {bShouldExcludeFromBranchCulling=true, weights [1,1,1]}}
 {
  id=SoundClass_3
  children [Attenuation_3]
  properties {soundClassOverride=&Fort_WeaponRanged_Fire_3P_PSM}
 }
 {
  id=SoundClass_4
  children [Attenuation_4]
  properties {soundClassOverride=&Fort_WeaponRanged_Fire_Distant}
 }
 {
  id=SoundClass_5
  children [Attenuation_5]
  properties {soundClassOverride=&Fort_WeaponRanged_Fire_Distant}
 }
 {
  id=WavePlayer_0
  properties {soundWaveAssetPtr=&/StickyGrenadeLauncher_Shoot_3P_Close_01}
 }
 {
  id=WavePlayer_1
  properties {soundWaveAssetPtr=&/StickyGrenadeLauncher_Shoot_3P_Close_02}
 }
 {
  id=WavePlayer_10
  properties {soundWaveAssetPtr=&/StickyGrenadeLauncher_Shoot_3P_Medium_01}
 }
 {
  id=WavePlayer_11
  properties {soundWaveAssetPtr=&/StickyGrenadeLauncher_Shoot_3P_Medium_02}
 }
 {
  id=WavePlayer_12
  properties {soundWaveAssetPtr=&/StickyGrenadeLauncher_Shoot_3P_Medium_03}
 }
 {
  id=WavePlayer_13
  properties {soundWaveAssetPtr=&/StickyGrenadeLauncher_Shoot_3P_Medium_04}
 }
 {
  id=WavePlayer_14
  properties {soundWaveAssetPtr=&/StickyGrenadeLauncher_Shoot_3P_Medium_05}
 }
 {
  id=WavePlayer_15
  properties {soundWaveAssetPtr=&/StickyGrenadeLauncher_Shoot_3P_Medium_06}
 }
 {
  id=WavePlayer_16
  properties {soundWaveAssetPtr=&/StickyGrenadeLauncher_Shoot_3P_Medium_07}
 }
 {
  id=WavePlayer_17
  properties {soundWaveAssetPtr=&/StickyGrenadeLauncher_Shoot_3P_Medium_08}
 }
 {
  id=WavePlayer_18
  properties {soundWaveAssetPtr=&/StickyGrenadeLauncher_Shoot_3P_Medium_09}
 }
 {
  id=WavePlayer_19
  properties {soundWaveAssetPtr=&/StickyGrenadeLauncher_Shoot_3P_Medium_10}
 }
 {
  id=WavePlayer_2
  properties {soundWaveAssetPtr=&/StickyGrenadeLauncher_Shoot_3P_Close_03}
 }
 {
  id=WavePlayer_20
  properties {soundWaveAssetPtr=&/StickyGrenadeLauncher_Shoot_3P_Distant_01}
 }
 {
  id=WavePlayer_21
  properties {soundWaveAssetPtr=&/StickyGrenadeLauncher_Shoot_3P_Distant_02}
 }
 {
  id=WavePlayer_22
  properties {soundWaveAssetPtr=&/StickyGrenadeLauncher_Shoot_3P_Distant_03}
 }
 {
  id=WavePlayer_23
  properties {soundWaveAssetPtr=&/StickyGrenadeLauncher_Shoot_3P_Distant_04}
 }
 {
  id=WavePlayer_24
  properties {soundWaveAssetPtr=&/StickyGrenadeLauncher_Shoot_3P_Distant_05}
 }
 {
  id=WavePlayer_25
  properties {soundWaveAssetPtr=&/StickyGrenadeLauncher_Shoot_3P_Distant_06}
 }
 {
  id=WavePlayer_26
  properties {soundWaveAssetPtr=&/StickyGrenadeLauncher_Shoot_3P_Distant_07}
 }
 {
  id=WavePlayer_27
  properties {soundWaveAssetPtr=&/StickyGrenadeLauncher_Shoot_3P_Distant_08}
 }
 {
  id=WavePlayer_28
  properties {soundWaveAssetPtr=&/StickyGrenadeLauncher_Shoot_3P_Distant_09}
 }
 {
  id=WavePlayer_29
  properties {soundWaveAssetPtr=&/StickyGrenadeLauncher_Shoot_3P_Distant_10}
 }
 {
  id=WavePlayer_3
  properties {soundWaveAssetPtr=&/StickyGrenadeLauncher_Shoot_3P_Close_04}
 }
 {
  id=WavePlayer_4
  properties {soundWaveAssetPtr=&/StickyGrenadeLauncher_Shoot_3P_Close_05}
 }
 {
  id=WavePlayer_5
  properties {soundWaveAssetPtr=&/StickyGrenadeLauncher_Shoot_3P_Close_06}
 }
 {
  id=WavePlayer_6
  properties {soundWaveAssetPtr=&/StickyGrenadeLauncher_Shoot_3P_Close_07}
 }
 {
  id=WavePlayer_7
  properties {soundWaveAssetPtr=&/StickyGrenadeLauncher_Shoot_3P_Close_08}
 }
 {
  id=WavePlayer_8
  properties {soundWaveAssetPtr=&/StickyGrenadeLauncher_Shoot_3P_Close_09}
 }
 {
  id=WavePlayer_9
  properties {soundWaveAssetPtr=&/StickyGrenadeLauncher_Shoot_3P_Close_10}
 }
]
properties {
 bDebug=true
 maxDistance=27500
 soundClassObject=&Fort_WeaponRanged_Fire_3P_PSM
 volumeMultiplier=1
 concurrencyOverrides {bLimitToOwner=true, maxCount=3, resolutionRule=StopOldest}
 concurrencySet [&Weapon_Gunfire_Global_Conc, &Weapon_Gunfire_Local_Conc]
}
