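# Sound cue for a sticky grenade launcher projectile hitting stone, routed through the impacts sound class with bullet-impact attenuation and a rifle bullet impact concurrency set.
# A short random delay leads into random pitch and volume modulation, then a quality-level switch.
# The first two quality slots pick randomly from all eight stone impact waves, while the last two share a picker limited to the final three waves.
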SoundCue /HopscotchWeaponsGameplay/Items/GrenadeLauncher/Sound/StickyGrenadeLauncher_Impact_Stone_Cue
root=Delay_1
nodes [
 {id=Delay_1, children [Modulator_0], properties {delayMax=0.13, delayMin=0.07}}
 {id=Modulator_0, children [QualityLevel_0], properties {pitchMax=1.25, pitchMin=1, volumeMax=1.1, volumeMin=0.8}}
 {id=QualityLevel_0, children [Random_4,Random_4,Random_5,Random_5]}
 {
  id=Random_4
  children [WavePlayer_0,WavePlayer_1,WavePlayer_2,WavePlayer_3,WavePlayer_4,WavePlayer_5,WavePlayer_6,WavePlayer_7]
  properties {weights [1,1,1,1,1,1,1,1]}
 }
 {id=Random_5, children [WavePlayer_5,WavePlayer_6,WavePlayer_7], properties {bShouldExcludeFromBranchCulling=true, weights [1,1,1]}}
 {id=WavePlayer_0, properties {soundWaveAssetPtr=&/StickyGrenadeLauncher_Impact_Stone_01}}
 {id=WavePlayer_1, properties {soundWaveAssetPtr=&/StickyGrenadeLauncher_Impact_Stone_02}}
 {id=WavePlayer_2, properties {soundWaveAssetPtr=&/StickyGrenadeLauncher_Impact_Stone_03}}
 {id=WavePlayer_3, properties {soundWaveAssetPtr=&/StickyGrenadeLauncher_Impact_Stone_04}}
 {id=WavePlayer_4, properties {soundWaveAssetPtr=&/StickyGrenadeLauncher_Impact_Stone_05}}
 {id=WavePlayer_5, properties {soundWaveAssetPtr=&/StickyGrenadeLauncher_Impact_Stone_06}}
 {id=WavePlayer_6, properties {soundWaveAssetPtr=&/StickyGrenadeLauncher_Impact_Stone_07}}
 {id=WavePlayer_7, properties {soundWaveAssetPtr=&/StickyGrenadeLauncher_Impact_Stone_08}}
]
properties {
 attenuationSettings=/Game/Sounds/Attenuation/Fort_Bullet_Impact_Att
 maxDistance=5300
 soundClassObject=&Fort_Impacts
 volumeMultiplier=1
 concurrencyOverrides {maxCount=4, resolutionRule=PreventNew}
 concurrencySet [&RifleBulletImpact_Conc]
}
