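# Sound cue for the sticky grenade launcher's low-ammo warning, using a degraded low-ammo ranged weapon sound class.
# A mixer layers a shared high rifle low-ammo wave with one of three randomly chosen launcher no-ammo clips, which pass through their own modulator for slight variation.
# The root modulator applies no pitch or volume variation.
# First-person weapon attenuation applies, and a concurrency override prevents new instances once the limit is reached.
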SoundCue /HopscotchWeaponsGameplay/Items/GrenadeLauncher/Sound/StickyGrenadeLauncher_LowAmmo_Cue
root=Modulator_0
nodes [
 {id=Mixer_0, children [WavePlayer_34, Modulator_1], properties {inputVolume [1,1]}}
 {id=Modulator_0, children [Mixer_0], properties {pitchMax=1, pitchMin=1, volumeMax=1, volumeMin=1}}
 {id=Modulator_1, children [Random_0], properties {pitchMax=1}}
 {id=Random_0, children [WavePlayer_0,WavePlayer_1,WavePlayer_2], properties {weights [1,1,1]}}
 {id=WavePlayer_0, properties {soundWaveAssetPtr=&/StickyGrenadeLauncher_NoAmmo_01}}
 {id=WavePlayer_1, properties {soundWaveAssetPtr=&/StickyGrenadeLauncher_NoAmmo_02}}
 {id=WavePlayer_2, properties {soundWaveAssetPtr=&/StickyGrenadeLauncher_NoAmmo_03}}
 {id=WavePlayer_34, properties {soundWaveAssetPtr=/Game/Sounds/Fort_Weapons/LowAmmo/weapon_low_ammo_hi_rifle_05}}
]
properties {
 attenuationSettings=&Fort_Weapon_Projectile_FirstPerson_Athena_Att
 bOverrideConcurrency=true
 maxDistance=2097152
 soundClassObject=/Game/Sounds/FortSoundClassesAndModes/Fort_WeaponRanged_LowAmmo_Degraded_PSM
 volumeMultiplier=1
 concurrencyOverrides {resolutionRule=PreventNew}
}
