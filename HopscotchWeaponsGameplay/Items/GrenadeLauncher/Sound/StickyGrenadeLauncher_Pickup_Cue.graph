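# Sound cue for picking up the Sticky Grenade Launcher.
# A Random node picks one of three equally weighted pickup waves and feeds a Modulator at default settings for slight variation.
# The cue plays at lowered pitch and volume in the loot-collect sound class, using shared weapon-pickup attenuation and a pickup concurrency set to limit overlapping sounds.
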SoundCue /HopscotchWeaponsGameplay/Items/GrenadeLauncher/Sound/StickyGrenadeLauncher_Pickup_Cue
root=Modulator_0
nodes [
 {id=Modulator_0, children [Random_0]}
 {id=Random_0, children [WavePlayer_0,WavePlayer_1,WavePlayer_2], properties {weights [1,1,1]}}
 {id=WavePlayer_0, properties {soundWaveAssetPtr=&/StickyGrenadeLauncher_PickUp_01}}
 {id=WavePlayer_1, properties {soundWaveAssetPtr=&/StickyGrenadeLauncher_PickUp_02}}
 {id=WavePlayer_2, properties {soundWaveAssetPtr=&/StickyGrenadeLauncher_PickUp_03}}
]
properties {
 attenuationSettings=&Fort_Weapon_Pickup_Att
 maxDistance=3500
 pitchMultiplier=0.9
 soundClassObject=&Fort_Loot_Collect
 volumeMultiplier=0.7
 concurrencyOverrides {maxCount=1, resolutionRule=PreventNew}
 concurrencySet [&PickupMaximum_Oldest]
}
