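# Sound cue for the sticky grenade launcher projectile's countdown sound.
# A short fixed delay feeds a single wave player playing the countdown wave.
# It plays through the Fort_Weapons sound class, with distance falloff from a dedicated countdown attenuation asset.
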SoundCue /HopscotchWeaponsGameplay/Items/GrenadeLauncher/Sound/StickyGrenadeLauncher_Proj_Countdown
root=Delay_0
nodes [
 {id=Delay_0, children [WavePlayer_0], properties {delayMax=0.15, delayMin=0.15}}
 {id=WavePlayer_0, properties {soundWaveAssetPtr=&/StickyGrenadeLauncher_Countdown_01}}
]
properties {
 attenuationSettings=/HopscotchWeaponsGameplay/Items/GrenadeLauncher/Sound/Settings/StickyGrenadeLauncher_Countdown_Att
 maxDistance=4500
 soundClassObject=/Game/Sounds/FortSoundClassesAndModes/Fort_Weapons
 volumeMultiplier=1
}
