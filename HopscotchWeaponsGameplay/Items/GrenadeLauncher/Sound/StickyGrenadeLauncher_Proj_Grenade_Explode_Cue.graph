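# Sound cue for the Sticky Grenade Launcher's grenade explosion, layered by listener distance.
# A root mixer blends close, medium and distant attenuation branches, with randomized waves and modulators varying each layer.
# Close combines blast, sweetener and delayed, separately attenuated debris; medium pitches up distant blasts and sweeteners; distant adds tails to distant blasts.
# Quality-level nodes switch between full and reduced variant pools, and concurrency stops the oldest instances.
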SoundCue /HopscotchWeaponsGameplay/Items/GrenadeLauncher/Sound/StickyGrenadeLauncher_Proj_Grenade_Explode_Cue
root=Mixer_4
nodes [
 {
  id=Attenuation_0
  children [Delay_0]
  properties {
   attenuationSettings=/HopscotchWeaponsGameplay/Items/GrenadeLauncher/Sound/Settings/StickyGrenadeLauncher_Debris_Explosion_Close_Att
   attenuationOverrides {falloffDistance=1600}
  }
 }
 {
  id=Attenuation_1
  children [Mixer_0]
  properties {attenuationSettings=/Game/Sounds/Attenuation/Fort_Explosion_Close_Athena_Att, attenuationOverrides {falloffDistance=1600}}
 }
 {
  id=Attenuation_2
  children [Mixer_2]
  properties {attenuationSettings=/Game/Sounds/Attenuation/Fort_Explosion_Distant_Athena_Att, attenuationOverrides {falloffDistance=1600}}
 }
 {
  id=Attenuation_8
  children [Mixer_1]
  properties {attenuationSettings=/Game/Sounds/Attenuation/Fort_Explosion_Medium_Athena_Att, attenuationOverrides {falloffDistance=1600}}
 }
 {id=Delay_0, children [Modulator_3], properties {delayMax=0.15, delayMin=0.15}}
 {id=Mixer_0, children [Modulator_2, Modulator_0, Attenuation_0], properties {inputVolume [1,1,0.7]}}
 {id=Mixer_1, children [Modulator_8,Modulator_1], properties {inputVolume [0.4,2]}}
 {id=Mixer_2, children [Modulator_4,Modulator_6], properties {inputVolume [0.3,1.65]}}
 {id=Mixer_4, children [Attenuation_1, Attenuation_8, Attenuation_2], properties {inputVolume [2.5,2,6]}}
 {id=Modulator_0, children [Random_0]}
 {id=Modulator_1, children [Random_1], properties {pitchMax=1.25, pitchMin=1.1}}
 {id=Modulator_2, children [QualityLevel_0]}
 {id=Modulator_3, children [QualityLevel_1], properties {pitchMax=1, pitchMin=1}}
 {id=Modulator_4, children [Random_2], properties {pitchMax=1.1, pitchMin=0.7}}
 {id=Modulator_6, children [QualityLevel_2]}
 {id=Modulator_8, children [QualityLevel_6], properties {pitchMax=1.4, pitchMin=1.2}}
 {id=QualityLevel_0, children [Random_3,Random_3,Random_8,Random_8]}
 {id=QualityLevel_1, children [Random_5,Random_5,Random_4,Random_4]}
 {id=QualityLevel_2, children [Random_10,Random_10,Random_11,Random_11]}
 {id=QualityLevel_6, children [Random_6,Random_6,Random_7,Random_7]}
 {id=Random_0, children [WavePlayer_20, WavePlayer_21, WavePlayer_22], properties {weights [1,1,1]}}
 {id=Random_1, children [WavePlayer_23, WavePlayer_24, WavePlayer_25], properties {bShouldExcludeFromBranchCulling=true, weights [1,1,1]}}
 {
  id=Random_10
  children [
   WavePlayer_39
   WavePlayer_40
   WavePlayer_41
   WavePlayer_42
   WavePlayer_43
   WavePlayer_44
   WavePlayer_45
   WavePlayer_46
   WavePlayer_47
   WavePlayer_48
  ]
  properties {bShouldExcludeFromBranchCulling=true, weights [1,1,1,1,1,1,1,1,1,1]}
 }
 {id=Random_11, children [WavePlayer_46, WavePlayer_47, WavePlayer_48], properties {bShouldExcludeFromBranchCulling=true, weights [1,1,1]}}
 {
  id=Random_2
  children [WavePlayer_32, WavePlayer_33, WavePlayer_34, WavePlayer_35]
  properties {bShouldExcludeFromBranchCulling=true, weights [1,1,1,1]}
 }
 {
  id=Random_3
  children [WavePlayer_0,WavePlayer_1,WavePlayer_2,WavePlayer_3,WavePlayer_4,WavePlayer_5,WavePlayer_6,WavePlayer_7,WavePlayer_8,WavePlayer_9]
  properties {weights [1,1,1,1,1,1,1,1,1,1]}
 }
 {id=Random_4, children [WavePlayer_29, WavePlayer_30, WavePlayer_31], properties {bShouldExcludeFromBranchCulling=true, weights [1,1,1]}}
 {
  id=Random_5
  children [WavePlayer_26, WavePlayer_27, WavePlayer_28, WavePlayer_29, WavePlayer_30, WavePlayer_31]
  properties {weights [1,1,1,1,1,1]}
 }
 {
  id=Random_6
  children [
   WavePlayer_10
   WavePlayer_11
   WavePlayer_12
   WavePlayer_13
   WavePlayer_14
   WavePlayer_15
   WavePlayer_16
   WavePlayer_17
   WavePlayer_18
   WavePlayer_19
  ]
  properties {bShouldExcludeFromBranchCulling=true, weights [1,1,1,1,1,1,1,1,1,1]}
 }
 {id=Random_7, children [WavePlayer_17, WavePlayer_18, WavePlayer_19], properties {bShouldExcludeFromBranchCulling=true, weights [1,1,1]}}
 {id=Random_8, children [WavePlayer_7,WavePlayer_8,WavePlayer_9], properties {bShouldExcludeFromBranchCulling=true, weights [1,1,1]}}
 {id=WavePlayer_0, properties {soundWaveAssetPtr=&/StickyGrenadeLauncher_Explode_Close_01}}
 {id=WavePlayer_1, properties {soundWaveAssetPtr=&/StickyGrenadeLauncher_Explode_Close_02}}
 {
  id=WavePlayer_10
  properties {soundWaveAssetPtr=&/StickyGrenadeLauncher_Explode_Distant_01}
 }
 {
  id=WavePlayer_11
  properties {soundWaveAssetPtr=&/StickyGrenadeLauncher_Explode_Distant_02}
 }
 {
  id=WavePlayer_12
  properties {soundWaveAssetPtr=&/StickyGrenadeLauncher_Explode_Distant_03}
 }
 {
  id=WavePlayer_13
  properties {soundWaveAssetPtr=&/StickyGrenadeLauncher_Explode_Distant_04}
 }
 {
  id=WavePlayer_14
  properties {soundWaveAssetPtr=&/StickyGrenadeLauncher_Explode_Distant_05}
 }
 {
  id=WavePlayer_15
  properties {soundWaveAssetPtr=&/StickyGrenadeLauncher_Explode_Distant_06}
 }
 {
  id=WavePlayer_16
  properties {soundWaveAssetPtr=&/StickyGrenadeLauncher_Explode_Distant_07}
 }
 {
  id=WavePlayer_17
  properties {soundWaveAssetPtr=&/StickyGrenadeLauncher_Explode_Distant_08}
 }
 {
  id=WavePlayer_18
  properties {soundWaveAssetPtr=&/StickyGrenadeLauncher_Explode_Distant_09}
 }
 {
  id=WavePlayer_19
  properties {soundWaveAssetPtr=&/StickyGrenadeLauncher_Explode_Distant_10}
 }
 {id=WavePlayer_2, properties {soundWaveAssetPtr=&/StickyGrenadeLauncher_Explode_Close_03}}
 {
  id=WavePlayer_20
  properties {soundWaveAssetPtr=&/StickyGrenadeLauncher_ExplodeSweetener_Close_01}
 }
 {
  id=WavePlayer_21
  properties {soundWaveAssetPtr=&/StickyGrenadeLauncher_ExplodeSweetener_Close_02}
 }
 {
  id=WavePlayer_22
  properties {soundWaveAssetPtr=&/StickyGrenadeLauncher_ExplodeSweetener_Close_03}
 }
 {
  id=WavePlayer_23
  properties {soundWaveAssetPtr=&/StickyGrenadeLauncher_ExplodeSweetener_Dist_01}
 }
 {
  id=WavePlayer_24
  properties {soundWaveAssetPtr=&/StickyGrenadeLauncher_ExplodeSweetener_Dist_02}
 }
 {
  id=WavePlayer_25
  properties {soundWaveAssetPtr=&/StickyGrenadeLauncher_ExplodeSweetener_Dist_03}
 }
 {
  id=WavePlayer_26
  properties {soundWaveAssetPtr=&/StickyGrenadeLauncher_Explosion__Debris_01}
 }
 {
  id=WavePlayer_27
  properties {soundWaveAssetPtr=&/StickyGrenadeLauncher_Explosion__Debris_02}
 }
 {
  id=WavePlayer_28
  properties {soundWaveAssetPtr=&/StickyGrenadeLauncher_Explosion__Debris_03}
 }
 {
  id=WavePlayer_29
  properties {soundWaveAssetPtr=&/StickyGrenadeLauncher_Explosion__Debris_04}
 }
 {id=WavePlayer_3, properties {soundWaveAssetPtr=&/StickyGrenadeLauncher_Explode_Close_04}}
 {
  id=WavePlayer_30
  properties {soundWaveAssetPtr=&/StickyGrenadeLauncher_Explosion__Debris_05}
 }
 {
  id=WavePlayer_31
  properties {soundWaveAssetPtr=&/StickyGrenadeLauncher_Explosion__Debris_06}
 }
 {
  id=WavePlayer_32
  properties {soundWaveAssetPtr=&/1StickyGrenadeLauncher_Distant_Tail_01}
 }
 {
  id=WavePlayer_33
  properties {soundWaveAssetPtr=&/1StickyGrenadeLauncher_Distant_Tail_02}
 }
 {
  id=WavePlayer_34
  properties {soundWaveAssetPtr=&/1StickyGrenadeLauncher_Distant_Tail_03}
 }
 {
  id=WavePlayer_35
  properties {soundWaveAssetPtr=&/1StickyGrenadeLauncher_Distant_Tail_04}
 }
 {
  id=WavePlayer_39
  properties {soundWaveAssetPtr=&StickyGrenadeLauncher_Explode_Distant_01}
 }
 {id=WavePlayer_4, properties {soundWaveAssetPtr=&/StickyGrenadeLauncher_Explode_Close_05}}
 {
  id=WavePlayer_40
  properties {soundWaveAssetPtr=&StickyGrenadeLauncher_Explode_Distant_02}
 }
 {
  id=WavePlayer_41
  properties {soundWaveAssetPtr=&StickyGrenadeLauncher_Explode_Distant_03}
 }
 {
  id=WavePlayer_42
  properties {soundWaveAssetPtr=&StickyGrenadeLauncher_Explode_Distant_04}
 }
 {
  id=WavePlayer_43
  properties {soundWaveAssetPtr=&StickyGrenadeLauncher_Explode_Distant_05}
 }
 {
  id=WavePlayer_44
  properties {soundWaveAssetPtr=&StickyGrenadeLauncher_Explode_Distant_06}
 }
 {
  id=WavePlayer_45
  properties {soundWaveAssetPtr=&StickyGrenadeLauncher_Explode_Distant_07}
 }
 {
  id=WavePlayer_46
  properties {soundWaveAssetPtr=&StickyGrenadeLauncher_Explode_Distant_08}
 }
 {
  id=WavePlayer_47
  properties {soundWaveAssetPtr=&StickyGrenadeLauncher_Explode_Distant_09}
 }
 {
  id=WavePlayer_48
  properties {soundWaveAssetPtr=&StickyGrenadeLauncher_Explode_Distant_10}
 }
 {id=WavePlayer_5, properties {soundWaveAssetPtr=&/StickyGrenadeLauncher_Explode_Close_06}}
 {id=WavePlayer_6, properties {soundWaveAssetPtr=&/StickyGrenadeLauncher_Explode_Close_07}}
 {id=WavePlayer_7, properties {soundWaveAssetPtr=&/StickyGrenadeLauncher_Explode_Close_08}}
 {id=WavePlayer_8, properties {soundWaveAssetPtr=&/StickyGrenadeLauncher_Explode_Close_09}}
 {id=WavePlayer_9, properties {soundWaveAssetPtr=&/StickyGrenadeLauncher_Explode_Close_10}}
]
properties {
 maxDistance=27500
 soundClassObject=&Fort_WeaponRanged_Fire_3P_PSM
 volumeMultiplier=1
 concurrencyOverrides {maxCount=4, resolutionRule=StopOldest, volumeScale=0.5}
 concurrencySet [/Game/Sounds/Concurrency_Rules/Weapon_Explosion_Conc]
}
