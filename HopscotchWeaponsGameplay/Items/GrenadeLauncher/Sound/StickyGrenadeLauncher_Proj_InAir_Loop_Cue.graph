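# Sound cue for a sticky grenade launcher projectile in flight, pairing an outgoing launch shot with looping timer and fuse layers.
# After a short fixed delay, a mixer plays both branches, each with its own attenuation settings.
# The launch branch randomizes volume slightly and picks a random variation, with a quality-level switch selecting the six-wave or reduced three-wave pool.
# An enveloper fades in the mixed loops, and a concurrency override limits simultaneous instances.
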SoundCue /HopscotchWeaponsGameplay/Items/GrenadeLauncher/Sound/StickyGrenadeLauncher_Proj_InAir_Loop_Cue
root=Delay_2
nodes [
 {
  id=Attenuation_0
  children [Enveloper_0]
  properties {attenuationSettings=/HopscotchWeaponsGameplay/Items/GrenadeLauncher/Sound/Settings/StickyGrenadeLauncher_InAir_Loop_Att}
 }
 {
  id=Attenuation_1
  children [Modulator_1]
  properties {attenuationSettings=/HopscotchWeaponsGameplay/Items/GrenadeLauncher/Sound/Settings/StickyGrenadeLauncher_InAir_Launch_Att}
 }
 {id=Delay_2, children [Mixer_0], properties {delayMax=0.1, delayMin=0.1}}
 {
  id=Enveloper_0
  children [Modulator_0]
  properties {
   pitchCurve {editorCurveData {keys [{time=0}]}}
   volumeCurve {editorCurveData {keys [<at 0 0 RCIM_Cubic>, {arriveTangent=3.4285, interpMode=RCIM_Cubic, leaveTangent=3.4285, tangentMode=RCTM_User, time=0.75, value=1}]}}
  }
 }
 {id=Mixer_0, children [Attenuation_1, Attenuation_0], properties {inputVolume [1.5,2]}}
 {id=Mixer_2, children [WavePlayer_6,WavePlayer_7], properties {inputVolume [1,1]}}
 {id=Modulator_0, children [Mixer_2], properties {pitchMax=1, pitchMin=1, volumeMax=1, volumeMin=1}}
 {id=Modulator_1, children [QualityLevel_0], properties {volumeMax=1, volumeMin=0.9}}
 {id=QualityLevel_0, children [Random_0,Random_0,Random_1,Random_1]}
 {id=Random_0, children [WavePlayer_2,WavePlayer_3,WavePlayer_4,WavePlayer_5,WavePlayer_8,WavePlayer_9], properties {weights [1,1,1,1,1,1]}}
 {id=Random_1, children [WavePlayer_5,WavePlayer_8,WavePlayer_9], properties {bShouldExcludeFromBranchCulling=true, weights [1,1,1]}}
 {id=WavePlayer_2, properties {soundWaveAssetPtr=&/StickyGrenadeLauncher_Outgoing_01}}
 {id=WavePlayer_3, properties {soundWaveAssetPtr=&/StickyGrenadeLauncher_Outgoing_02}}
 {id=WavePlayer_4, properties {soundWaveAssetPtr=&/StickyGrenadeLauncher_Outgoing_03}}
 {id=WavePlayer_5, properties {soundWaveAssetPtr=&/StickyGrenadeLauncher_Outgoing_04}}
 {
  id=WavePlayer_6
  properties {bLooping=true, soundWaveAssetPtr=&/StickyGrenadeLauncher_InAir_Timer_Loop}
 }
 {
  id=WavePlayer_7
  properties {bLooping=true, soundWaveAssetPtr=&/StickyGrenadeLauncher_InAir_Fuse_Loop}
 }
 {id=WavePlayer_8, properties {soundWaveAssetPtr=&/StickyGrenadeLauncher_Outgoing_05}}
 {id=WavePlayer_9, properties {soundWaveAssetPtr=&/StickyGrenadeLauncher_Outgoing_06}}
]
properties {
 bOverrideConcurrency=true
 maxDistance=12500
 soundClassObject=&Fort_WeaponRanged_Projectile_Effects
 volumeMultiplier=1
 attenuationOverrides {nonSpatializedRadiusStart=200, stereoSpread=500}
 concurrencyOverrides {maxCount=3}
}
