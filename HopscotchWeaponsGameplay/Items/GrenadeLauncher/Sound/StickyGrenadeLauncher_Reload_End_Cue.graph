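# Sound cue for the end of the sticky grenade launcher's reload, played as weapon foley with reload attenuation.
# A mixer layers a randomized reload-end sound at varied, reduced volume, a randomized close sound, and a single cartridge sound on a short fixed delay.
# Quality-level switches give the reload-end and close layers full variation pools in the first two slots and smaller subsets in the last two.
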SoundCue /HopscotchWeaponsGameplay/Items/GrenadeLauncher/Sound/StickyGrenadeLauncher_Reload_End_Cue
root=Mixer_0
nodes [
 {id=Delay_0, children [Modulator_1], properties {delayMax=0.2, delayMin=0.2}}
 {id=Mixer_0, children [Modulator_0, QualityLevel_0, Delay_0], properties {inputVolume [1,1,0.8]}}
 {id=Modulator_0, children [QualityLevel_1], properties {pitchMax=1, pitchMin=1, volumeMax=0.6, volumeMin=0.5}}
 {id=Modulator_1, children [WavePlayer_6]}
 {id=QualityLevel_0, children [Random_0,Random_0,Random_3,Random_3]}
 {id=QualityLevel_1, children [Random_1,Random_1,Random_2,Random_2]}
 {id=Random_0, children [WavePlayer_7, WavePlayer_8, WavePlayer_9, WavePlayer_10, WavePlayer_11], properties {weights [1,1,1,1,1]}}
 {id=Random_1, children [WavePlayer_0,WavePlayer_1,WavePlayer_2,WavePlayer_3,WavePlayer_4,WavePlayer_5], properties {weights [1,1,1,1,1,1]}}
 {id=Random_2, children [WavePlayer_3,WavePlayer_4,WavePlayer_5], properties {bShouldExcludeFromBranchCulling=true, weights [1,1,1]}}
 {id=Random_3, children [WavePlayer_9, WavePlayer_10, WavePlayer_11], properties {bShouldExcludeFromBranchCulling=true, weights [1,1,1]}}
 {id=WavePlayer_0, properties {soundWaveAssetPtr=&/StickyGrenadeLauncher_Reload_End_01}}
 {id=WavePlayer_1, properties {soundWaveAssetPtr=&/StickyGrenadeLauncher_Reload_End_02}}
 {id=WavePlayer_10, properties {soundWaveAssetPtr=&/StickyGrenadeLauncher_Close_04}}
 {id=WavePlayer_11, properties {soundWaveAssetPtr=&/StickyGrenadeLauncher_Close_05}}
 {id=WavePlayer_2, properties {soundWaveAssetPtr=&/StickyGrenadeLauncher_Reload_End_03}}
 {id=WavePlayer_3, properties {soundWaveAssetPtr=&/StickyGrenadeLauncher_Reload_End_04}}
 {id=WavePlayer_4, properties {soundWaveAssetPtr=&/StickyGrenadeLauncher_Reload_End_05}}
 {id=WavePlayer_5, properties {soundWaveAssetPtr=&/StickyGrenadeLauncher_Reload_End_06}}
 {
  id=WavePlayer_6
  properties {soundWaveAssetPtr=&/StickyGrenadeLauncher_Reload_Cartridge_05}
 }
 {id=WavePlayer_7, properties {soundWaveAssetPtr=&/StickyGrenadeLauncher_Close_01}}
 {id=WavePlayer_8, properties {soundWaveAssetPtr=&/StickyGrenadeLauncher_Close_02}}
 {id=WavePlayer_9, properties {soundWaveAssetPtr=&/StickyGrenadeLauncher_Close_03}}
]
properties {
 attenuationSettings=&Fort_Weapon_Reload_Att
 bDebug=true
 bOverrideConcurrency=true
 maxDistance=2250
 soundClassObject=&Fort_WeaponFoley
 volumeMultiplier=0.752
}
