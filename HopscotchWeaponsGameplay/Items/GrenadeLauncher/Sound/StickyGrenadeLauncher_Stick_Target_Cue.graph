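# Sound cue for the sticky grenade launcher's projectile sticking to a target.
# A mixer layers a pitch-modulated landing sound, a randomized stick impact and a louder arm-timer sound, each through its own modulator.
# A quality-level node switches the impact between a five-variant and a three-variant random pool.
# Overrides supply logarithmic distance attenuation with low-pass filtering and limit concurrency to one instance.
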SoundCue /HopscotchWeaponsGameplay/Items/GrenadeLauncher/Sound/StickyGrenadeLauncher_Stick_Target_Cue
root=Mixer_0
nodes [
 {id=Mixer_0, children [Modulator_1,Modulator_0,Modulator_2], properties {inputVolume [0.8,2.2,3]}}
 {id=Modulator_0, children [QualityLevel_0], properties {pitchMax=1, pitchMin=0.9}}
 {id=Modulator_1, children [WavePlayer_7], properties {pitchMax=0.9, pitchMin=0.85}}
 {id=Modulator_2, children [WavePlayer_0]}
 {id=QualityLevel_0, children [Random_0,Random_0,Random_1,Random_1]}
 {id=Random_0, children [WavePlayer_1,WavePlayer_3,WavePlayer_4,WavePlayer_5,WavePlayer_6], properties {weights [1,1,1,1,1]}}
 {id=Random_1, children [WavePlayer_4,WavePlayer_5,WavePlayer_6], properties {bShouldExcludeFromBranchCulling=true, weights [1,1,1]}}
 {id=WavePlayer_0, properties {soundWaveAssetPtr=&/StickyGrenadeLauncher_Land_ArmTimer}}
 {id=WavePlayer_1, properties {soundWaveAssetPtr=&/StickyGrenadeLauncher_Stick_Target_01}}
 {id=WavePlayer_3, properties {soundWaveAssetPtr=&/StickyGrenadeLauncher_Stick_Target_02}}
 {id=WavePlayer_4, properties {soundWaveAssetPtr=&/StickyGrenadeLauncher_Stick_Target_03}}
 {id=WavePlayer_5, properties {soundWaveAssetPtr=&/StickyGrenadeLauncher_Stick_Target_04}}
 {id=WavePlayer_6, properties {soundWaveAssetPtr=&/StickyGrenadeLauncher_Stick_Target_05}}
 {id=WavePlayer_7, properties {soundWaveAssetPtr=&/StickyGrenadeLauncher_Land}}
]
properties {
 attenuationSettings=/Game/Sounds/Attenuation/Fort_Grenade_Bounce_Att
 bOverrideAttenuation=true
 bOverrideConcurrency=true
 maxDistance=3000
 soundClassObject=&Fort_WeaponRanged_Projectile_Effects
 volumeMultiplier=1
 attenuationOverrides {
  bAttenuateWithLPF=true
  distanceAlgorithm=Logarithmic
  falloffDistance=2500
  lPFFrequencyAtMax=2000
  lPFRadiusMax=3000
  lPFRadiusMin=500
  attenuationShapeExtents {x=500}
 }
 concurrencyOverrides {maxCount=1}
}
